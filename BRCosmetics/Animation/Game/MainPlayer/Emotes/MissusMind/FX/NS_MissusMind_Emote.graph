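# Niagara effect for the MissusMind player emote, flashing short-lived colored lights around the effect origin in two timed bursts.
# One Lights emitter spawns particle pairs, using a spawn group index to pick from four offsets, angled mesh orientations and red or gray colors.
# Each particle drives light, mesh and layered sprite renderers through a shared fade-in/out alpha curve.
# Mesh colors invert the red/gray pairing, and lights sit below their particles.
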NiagaraSystem /BRCosmetics/Animation/Game/MainPlayer/Emotes/MissusMind/FX/NS_MissusMind_Emote
dependencies {
 usedDataInterfaces [&NiagaraDataInterfaceArrayColor, &NiagaraDataInterfaceArrayFloat3, &NiagaraDataInterfaceCurve]
 usedDynamicInputs []
 usedModules [&SystemState]
}
emitters {Lights {
  emitterUpdateScript [
   &EmitterState()
   &SpawnBurst_Instantaneous(Spawn Count=2, Use Spawn Probability=false, UseLoopCountLimit=false)
   &SpawnRate(Use Spawn Probability=false) {enabled=false}
   &SpawnBurst_Instantaneous(Spawn Count=2, Spawn Time=1.3833, Use Spawn Probability=false, UseLoopCountLimit=false) {module=SpawnBurst_Instantaneous001}
  ]
  particleSpawnScript [
   &FN_SetSpawnGroup()
   &InitializeParticle(
    Color=&SelectLinearColorFromArray(
     Array Sampling Mode=::NewEnumerator1
     Direct Array Index=$Particles.SpawnGroup
     Color Selection Array=DataInterface() {propertyValues {ColorData [<rgba 1 0 0 1>, <rgba 0.5 0.5 0.5 1>, <rgba 1 0 0 1>, <rgba 0.5 0.5 0.5 1>]}}
     Random Seed=?
    )
    Lifetime=0.4
    Mesh Scale=<1 1 1.4>
    Mesh Scale Mode=::NewEnumerator4
    Position Offset=&SelectVectorFromArray(
     Direct Array Index=$Particles.SpawnGroup
     Vector Selection Array=DataInterface() {propertyValues {FloatData [<100 0 100>, <-100 0 100>, <0 100 100>, <0 -100 100>]}}
    )
    Sprite Size Mode=::NewEnumerator3
    Uniform Sprite Size=50
    UsePositionOffset=true
    Write Material Random=true
   )
   &InitialMeshOrientation(
    Mesh Axis To Orient=&SelectVectorFromArray(
     Direct Array Index=$Particles.SpawnGroup
     Vector Selection Array=DataInterface() {propertyValues {FloatData [<1 0 1>, <1 0 -1>, <0 1 1>, <0 1 -1>]}}
    )
    Mesh Orientation Mode=::NewEnumerator2
    Orientation Vector=&SelectVectorFromArray(
     Direct Array Index=$Particles.SpawnGroup
     Vector Selection Array=DataInterface() {propertyValues {FloatData [<1 0 0>, <1 0 0>, <0 1 0>, <0 1 0>]}}
    )
    Use Rotation Vector=false
   )
  ]
  particleUpdateScript [
   &ParticleState()
   /SetVariables(Particles.AlphaCurve:NiagaraFloat=&FloatFromCurve(FloatCurve=DataInterface() {propertyValues {Curve {keys [
         {interpMode=RCIM_Cubic, tangentMode=RCTM_User, time=0, value=0}
         {interpMode=RCIM_Cubic, tangentMode=RCTM_User, time=0.15, value=1}
         {arriveTangent=0.0026, interpMode=RCIM_Cubic, leaveTangent=0.0026, tangentMode=RCTM_User, time=0.85, value=1}
         {tangentMode=RCTM_User, time=1, value=0}
        ]}}})) {isSetParameters=true}
   &FN_Light_Attributes(
    Color=$Particles.Color
    Create Color Attribute=true
    Create Exponent Attribute=true
    Create Position Attribute=true
    Create Radius Attribute=true
    Exponent=5
    Position=&AddVectorToPosition(Position=$Particles.Position, Vector=<0 0 -100>)
    Radius=250
    Scale Color=&Multiply_Float(A=13, B=$Particles.AlphaCurve)
   )
   &FN_Mesh_Attributes(
    Color=&MultiplyLinearColor(
     Linear Color=&SelectLinearColorFromArray(
      Array Sampling Mode=::NewEnumerator1
      Direct Array Index=$Particles.SpawnGroup
      Color Selection Array=DataInterface() {propertyValues {ColorData [<rgba 0.5 0.5 0.5 1>, <rgba 1 0 0 1>, <rgba 0.5 0.5 0.5 1>, <rgba 1 0 0 1>]}}
      Random Seed=?
     )
     Scale Factor=&MakeVector4(X=1, Y=1, Z=1, W=&Multiply_Float(A=$Particles.AlphaCurve, B=0.75))
    )
    Create Color Attribute=true
    Write to Module Namespace=true
   )
   &FN_Sprite_Attributes(
    Color=&MultiplyLinearColor(
     Linear Color=$Particles.Color
     Scale Factor=&MakeVector4(X=1, Y=1, Z=1, W=&Multiply_Float(A=0.15, B=$Particles.AlphaCurve))
    )
    Create Color Attribute=true
    Create SpriteSize Attribute=true
    SpriteSize=<250 250>
    Write to Module Namespace=true
   )
   &FN_Sprite_Attributes(
    Color=&MultiplyLinearColor(Linear Color=$Particles.FN_Mesh_Attributes.Color, Scale Factor=<1 1 1 0.5>)
    Create Color Attribute=true
    Create SpriteSize Attribute=true
    SpriteSize=&Vector2DFromFloat(Value=170)
    Write to Module Namespace=true
   ) {module=FN_Sprite_Attributes001}
  ]
  rendererClasses [
   &NiagaraSpriteRendererProperties
   &NiagaraLightRendererProperties
   &NiagaraMeshRendererProperties
  ]
  renderers [
   {rendererClass=&NiagaraSpriteRendererProperties}
   {rendererClass=&NiagaraLightRendererProperties, rendererIndex=1}
   {rendererClass=&NiagaraMeshRendererProperties, rendererIndex=2}
   {rendererClass=&NiagaraSpriteRendererProperties, rendererIndex=3}
   {rendererClass=&NiagaraSpriteRendererProperties, rendererIndex=4}
  ]
 }}
summary {
 systemName=NS_MissusMind_Emote
 emitters [{
   emitterName=Lights
   rendererClasses [
    &NiagaraSpriteRendererProperties
    &NiagaraLightRendererProperties
    &NiagaraMeshRendererProperties
   ]
  }]
 userVariables []
}
